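# Looping spatialized mono MetaSound ambience for DragonCart Shard 3, with a send to the SkyMango fragment audio bus.
# The Default page mixes base and tonal loops, scales them by one minus a focus multiplier so the sound dips when faced, then applies MasterGain.
# The Mobile page plays only the base loop, low-passed and attenuated by the enclosure modulation bus, with the same focus ducking but no MasterGain.
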MetaSoundSource /DragonCartNarrative/Assets/Audio/Shard/MSS_DragonCart_Shard3_Single_Loop
inputs [
 Base Sound {type=WaveAsset}
 MasterGain {type=Float}
 Tonal Sound {type=WaveAsset}
 UE.Source.OnPlay {type=Trigger}
 UE.Spatialization.Azimuth {type=Float}
 UE.Spatialization.Elevation {type=Float}
]
interfaces [UE.OutputFormat.Mono:1.0, UE.Source:1.1, UE.Spatialization:1.0]
outputs ["UE.OutputFormat.Mono.Audio:0" {type=Audio}]
pages [
 Default {
  is_build_graph=true
  input_defaults {
   Base Sound=/DragonCartNarrative/Assets/Audio/Shard/wav/DragonCart_Shard_Base_LP
   MasterGain=3
   Tonal Sound=/DragonCartNarrative/Assets/Audio/Shard/wav/DragonCart_Shard3_Tonal_LP
   UE.Source.OnPlay=null
   UE.Spatialization.Azimuth=0
   UE.Spatialization.Elevation=0
  }
  nodes [
   WavePlayer_0 = UE.Wave Player:Mono(Wave Asset=^Base Sound, Play=^UE.Source.OnPlay) {also {defaults {Loop="<MetaSoundBuilderResult.SUCCEEDED: 0>", Pitch Shift=1}, version [1,1]}}
   WavePlayer_1 = UE.Wave Player:Mono(Wave Asset=^Tonal Sound, Play=^UE.Source.OnPlay) {also {defaults {Loop="<MetaSoundBuilderResult.SUCCEEDED: 0>"}, version [1,1]}}
   AudioMixerAudioMixerMono2_0 = "AudioMixer.Audio Mixer (Mono, 2)"(In 0=@WavePlayer_0.Out Mono, In 1=@WavePlayer_1.Out Mono) {also {defaults {Gain 0=0.5}}}
   Multiply_0 = UE.Multiply:Audio by Float(
    PrimaryOperand=@AudioMixerAudioMixerMono2_0.Out
    AdditionalOperands=UE.Subtract:Float(AdditionalOperands=/FortMetasound/Attenuation/MSP_Attenuation_Focus(Azimuth=^UE.Spatialization.Azimuth, Elevation=^UE.Spatialization.Elevation) {also {defaults {
        FocusAngleThreshold=15
        FocusAttackInterpSpeed=30
        FocusReleaseInterpSpeed=8
        FocusValue=0.7
        NonFocusAngleThreshold=30
        NonFocusValue=0
        bEnableFocusInterpolation="<MetaSoundBuilderResult.SUCCEEDED: 0>"
       }}}.FocusMultiplier) {also {defaults {PrimaryOperand=1}}}.Out
   )
   Multiply_1 = UE.Multiply:Audio by Float(AdditionalOperands=^MasterGain, PrimaryOperand=@Multiply_0.Out)
  ]
  links [{dst="@UE.OutputFormat.Mono.Audio:0", src=Multiply_1.Out}]
 }
 Mobile {
  nodes [
   n1 = Modulation.GetModulatorValue() {also {defaults {Modulator=&AudioWorldization_Bus_Enclosure}}}
   WavePlayer_0 = UE.Wave Player:Mono(Wave Asset=^Base Sound, Play=^UE.Source.OnPlay) {also {defaults {Loop="<MetaSoundBuilderResult.SUCCEEDED: 0>"}, version [1,1]}}
   BiquadFilter_0 = UE.Biquad Filter:Audio(
    Cutoff Frequency=MapRange.MapRange:Float(In=@n1.Out) {also {defaults {In Range A=0.85, Out Range A=20000, Out Range B=150}}}.Out Value
    In=@WavePlayer_0.Out Mono
   ) {also {defaults {Bandwidth=0.5, Type=0}}}
   Multiply_1 = UE.Multiply:Audio by Float(
    PrimaryOperand=@BiquadFilter_0.Out
    AdditionalOperands=MapRange.MapRange:Float(In=@n1.Out) {also {defaults {In Range A=0.85, Out Range A=1, Out Range B=0.2}}}.Out Value
   )
   Multiply_0 = UE.Multiply:Audio by Float(
    PrimaryOperand=@Multiply_1.Out
    AdditionalOperands=UE.Subtract:Float(AdditionalOperands=&MSP_Attenuation_Focus(Azimuth=^UE.Spatialization.Azimuth, Elevation=^UE.Spatialization.Elevation) {also {defaults {
        FocusAngleThreshold=15
        FocusAttackInterpSpeed=30
        FocusReleaseInterpSpeed=8
        FocusValue=0.7
        NonFocusAngleThreshold=30
        NonFocusValue=0
        bEnableFocusInterpolation="<MetaSoundBuilderResult.SUCCEEDED: 0>"
       }}}.FocusMultiplier) {also {defaults {PrimaryOperand=1}}}.Out
   )
  ]
  links [{dst="@UE.OutputFormat.Mono.Audio:0", src=Multiply_0.Out}]
 }
]
properties {
 assetImportData=&self:AssetImportData
 attenuationSettings=/Hera/Props/SkyMango/SkyMango_Fragment_A/Audio/ATT_SkyMango_Fragment_A
 bLooping=true
 duration=10000
 maxDistance=5500
 sampleRate=48000
 soundClassObject=&Fort_Ambient_Athena
 busSends [{audioBus=/Hera/Props/SkyMango/SkyMango_Fragment_A/Audio/AB_SkyMango_Fragment, sendLevel=0.2}]
}
