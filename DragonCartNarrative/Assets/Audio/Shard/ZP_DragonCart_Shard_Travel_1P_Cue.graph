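# Sound cue for first-person (1P) shard portal travel in DragonCart, layering an entry sound with a travel loop.
# A mixer plays the one-shot portal-enter wave alongside an enveloped looping portal-travel wave; its middle input is empty.
# The envelope ramps the loop's volume up from silence over about three seconds to a partial level.
# The cue primes on load and uses trap attenuation and the gadgets/items sound class.
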SoundCue /DragonCartNarrative/Assets/Audio/Shard/ZP_DragonCart_Shard_Travel_1P_Cue
root=Mixer_1
nodes [
 {
  id=Enveloper_1
  children [WavePlayer_3]
  properties {
   pitchCurve {editorCurveData {keys [{time=0}]}}
   volumeCurve {editorCurveData {keys [<at 0 0>,<at 1.0139 0.3482>,<at 2.6849 0.6767>,<at 3.0042 0.6786>]}}
  }
 }
 {id=Mixer_1, children [WavePlayer_5,null,Enveloper_1], properties {inputVolume [1,0.4,1]}}
 {id=WavePlayer_3, properties {bLooping=true, soundWaveAssetPtr=/FNGameplayCues/Athena/Sounds/Events/ZP2/ZP_Portal_Travel_Loop_01}}
 {id=WavePlayer_5, properties {soundWaveAssetPtr=&/ZP_DragonCart_Shard_Portal_Enter}}
]
properties {
 attenuationSettings=&Fort_Traps_Att
 bPrimeOnLoad=true
 maxDistance=6000
 soundClassObject=&Fort_Gadgets_Items
 volumeMultiplier=1
}
