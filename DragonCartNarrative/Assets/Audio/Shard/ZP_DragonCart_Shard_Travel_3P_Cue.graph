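# Third-person sound cue for Dragon Cart shard travel, built on the shard portal-enter sound wave.
# An Enveloper wraps the single wave player, holding full volume briefly and then fading it to silence by about 1.3 seconds.
# It uses the trap attenuation settings and the gadget-items sound class, and primes on load.
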SoundCue /DragonCartNarrative/Assets/Audio/Shard/ZP_DragonCart_Shard_Travel_3P_Cue
root=Enveloper_0
nodes [
 {
  id=Enveloper_0
  children [WavePlayer_5]
  properties {pitchCurve {editorCurveData {keys [{time=0}]}}, volumeCurve {editorCurveData {keys [<at 0 1>,<at 0.3 1>,<at 1.3 0>]}}}
 }
 {id=WavePlayer_5, properties {soundWaveAssetPtr=&/ZP_DragonCart_Shard_Portal_Enter}}
]
properties {
 attenuationSettings=&Fort_Traps_Att
 bPrimeOnLoad=true
 maxDistance=6000
 soundClassObject=&Fort_Gadgets_Items
 volumeMultiplier=1
}
